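; Blueprint function library of helpers that apply banner visuals (icon, colors, shape) to a dynamic material instance.
; GenericUpdateMaterial is the all-in-one call: it delegates to the icon and color helpers, then sets the shape texture from its input.
; It also forces a hardcoded HomeBase pattern texture and a white icon tint.
; The color helper writes primary and secondary background colors into the RGBA0 and RGBA1 parameters.
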(fn GenericUpdateMaterial
  (params (Target :type Material%20Instance%20Dynamic%20Object%20Reference) (BannerIcon :type Texture%20Object%20Reference) (BG_PrimaryColor :type Linear%20Color%20Structure) (BG_SecondaryColor :type Linear%20Color%20Structure) (ShapeIcon :type Texture%20Object%20Reference))
  (HelperMethods|UpdateBannerIcononMaterial (Variables|GetTarget) (Variables|GetBannerIcon) (Variables|Get__WorldContext))
  (Rendering|Material|SetTextureParameterValue (Variables|GetTarget) "BannerPattern" "/Game/UI/HomeBase/BannerWidget/HB-Banner-Pattern-2.HB-Banner-Pattern-2")
  (Rendering|Material|SetVectorParameterValue (Variables|GetTarget) "BannerIconColor" "(R=1.000000,G=1.000000,B=1.000000,A=1.000000)" :!class /Script/Engine.MaterialInstanceDynamic)
  (HelperMethods|UpdateBannerColoronMaterial (Variables|GetTarget) (Variables|GetBG_PrimaryColor) (Variables|GetBG_SecondaryColor) (Variables|Get__WorldContext))
  (Rendering|Material|SetTextureParameterValue (Variables|GetTarget) "BannerShape" (Variables|GetShapeIcon)))

(fn UpdateBannerColorOnMaterial
  (params (Material :type Material%20Instance%20Dynamic%20Object%20Reference) (PrimaryBGColor :type Linear%20Color%20Structure) (SecondaryBGColor :type Linear%20Color%20Structure))
  (Rendering|Material|SetVectorParameterValue Material "RGBA0" PrimaryBGColor :!class /Script/Engine.MaterialInstanceDynamic)
  (Rendering|Material|SetVectorParameterValue Material "RGBA1" SecondaryBGColor :!class /Script/Engine.MaterialInstanceDynamic))

(fn UpdateBannerIconOnMaterial
  (params (Material :type Material%20Instance%20Dynamic%20Object%20Reference) (Icon :type Texture%20Object%20Reference))
  (Rendering|Material|SetTextureParameterValue Material "BannerIcon" Icon))
